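; Fortnite movement mode extension for a boosted jump that charges the owning pawn energy when it activates.
; On activation it calls the parent logic, then looks up the pawn's energy component.
; If that component is valid, it spends a scalable-float energy cost evaluated in the pawn's context.
; The spend runs only on the server or for the local player.
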(event EventOnExtensionActivate
  (params (RuntimeContext :type MMERuntime%20Context%20Structure%20%28by%20ref%29))
  (bind _ownerpawn (Variables|FortMovementMode_BaseExtLogic|GetOwnerPawn))
  (bind _returnvalue (Actor|GetComponentbyClass _ownerpawn "/Script/EnergyRuntime.FortComponent_Energy"))
  (|CallParentFunction RuntimeContext)
  (Utilities|IsValid _returnvalue
    (:Is%20Valid
      (if (or (Variables|FortMovementMode_BaseExtLogic|GetbIsServer) (Variables|FortMovementMode_BaseExtLogic|GetbIsLocalPlayer))
        (Energy|UseEnergy _returnvalue (ScalableFloat|GetValueatLevelwithContext _ownerpawn (Variables|Default|GetBoostedJumpEnergyCost)) true 0)))))
